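# Physics asset for the CandyGirl Fortnite backpack cosmetic, defining collision bodies and joints for its skeletal mesh.
# Capsule bodies cover the backpack_root, belly and star bones.
# Both belly and star constrain directly to backpack_root.
# The belly joint locks all swing and twist, while the star joint leaves them unlocked, with uneven swing angles and a narrow twist angle.
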PhysicsAsset /BRCosmetics/Accessories/FORT_Backpacks/F_MED_CandyGirl_Backpack/Meshes/F_MED_CandyGirl_Backpack_Physics
bodies {
 backpack_root [{shapeType=Capsule, length=6.7502, radius=2.0262, center=<-3.4981 -6.3327 -0.5296>, rotation=<rot 0.4748 -90 -87.4663>}]
 belly [{shapeType=Capsule, length=0.505, radius=6.255, rotation=<rot 0 -90 0>}]
 star [{shapeType=Capsule, length=3.2551, radius=3.1673, center=<-1.2594 -0.1318 0.0524>, rotation=<rot 0 -84.2078 -86.7482>}]
}
constraints [
 {
  bone1Name=belly
  bone2Name=backpack_root
  swing1LimitDegrees=45
  swing1Motion=Locked
  swing2LimitDegrees=45
  swing2Motion=Locked
  twistLimitDegrees=45
  twistMotion=Locked
 }
 {bone1Name=star, bone2Name=backpack_root, swing1LimitDegrees=30, swing2LimitDegrees=60, twistLimitDegrees=5}
]
